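; Master Blueprint for water mesh actors that tracks player pawns in water and triggers splashes.
; Construction destroys surplus collision meshes and hides StaticMesh1 with a vertex-shrink material, since hidden meshes break traces.
; Begin/end overlaps on up to four volumes adjust a per-pawn counter that sets IsInWater.
; With interactions enabled, sufficient vertical speed and an expired cooldown, a component line trace places splashes played through the pawn and WindManager.
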(event EventBeginPlay
  (CallFunction|Construction))

(event Custom|Construct
  (CallFunction|Construction))

(event OnComponentBeginOverlap(CollisionMesh1)
  (params (OverlappedComponent :type Primitive%20Component%20Object%20Reference) (OtherActor :type Actor%20Object%20Reference) (OtherComp :type Primitive%20Component%20Object%20Reference) (OtherBodyIndex :type Integer) (bFromSweep :type Boolean) (SweepResult :type Hit%20Result%20Structure%20%28by%20ref%29))
  (bind _externalactor (Variables|Default|GetExternalActor))
  (CallFunction|HandleWaterPenetrationComponentsandArrays true OtherActor (Variables|Default|GetCollisionMesh1))
  ;[1 1 1] Early out
  (if (and (Variables|WaterSettings|GetEnable%20Player/Water%20Interactions) (Variables|Default|GetContinue))
    ;[0 0 0] Spawn A Splash Effect
    (if (and (> (Utilities|Time|GetGameTimeinSeconds) (Variables|/Game/Abilities/Player/Pawns/PlayerPawn_Generic_Parent.PlayerPawn_Generic_Parent_C|GetTime%20when%20you'll%20be%20able%20to%20splash%20again (Variables|Default|GetExternalActor))) (> (Math|Float|Absolute(Float) (.z (Transformation|GetVelocity (Variables|Default|GetExternalActor)))) 30))
      (bind (_traced_transform _velocityvector _trace_hit) (CallFunction|FindPlacementOnWaterMeshSurfaces))
      (if _trace_hit
        (Variables|/Game/Abilities/Player/Pawns/PlayerPawn_Generic_Parent.PlayerPawn_Generic_Parent_C|SetTime%20when%20you'll%20be%20able%20to%20splash%20again (+ (Utilities|Time|GetGameTimeinSeconds) (Variables|Default|GetMinimum%20time%20between%20splashes)) _externalactor)
        (Class|PlayerPawnGenericParent|PlayerCreatesASplash _externalactor _traced_transform)
        (Class|WindManager|PlayWaterSplashParticleSystematLocation (Utilities|Casting|CastToWindManager (Fortnite|GetWindManager)) _traced_transform)))))

(event OnComponentBeginOverlap(CollisionMesh2)
  (params (OverlappedComponent :type Primitive%20Component%20Object%20Reference) (OtherActor :type Actor%20Object%20Reference) (OtherComp :type Primitive%20Component%20Object%20Reference) (OtherBodyIndex :type Integer) (bFromSweep :type Boolean) (SweepResult :type Hit%20Result%20Structure%20%28by%20ref%29))
  (bind _externalactor (Variables|Default|GetExternalActor))
  (CallFunction|HandleWaterPenetrationComponentsandArrays true OtherActor (Variables|Default|GetCollisionMesh2))
  ;[1 1 1] Early out
  (if (and (Variables|WaterSettings|GetEnable%20Player/Water%20Interactions) (Variables|Default|GetContinue))
    ;[0 0 0] Spawn A Splash Effect
    (if (and (> (Utilities|Time|GetGameTimeinSeconds) (Variables|/Game/Abilities/Player/Pawns/PlayerPawn_Generic_Parent.PlayerPawn_Generic_Parent_C|GetTime%20when%20you'll%20be%20able%20to%20splash%20again (Variables|Default|GetExternalActor))) (> (Math|Float|Absolute(Float) (.z (Transformation|GetVelocity (Variables|Default|GetExternalActor)))) 30))
      (bind (_traced_transform _velocityvector _trace_hit) (CallFunction|FindPlacementOnWaterMeshSurfaces))
      (if _trace_hit
        (Variables|/Game/Abilities/Player/Pawns/PlayerPawn_Generic_Parent.PlayerPawn_Generic_Parent_C|SetTime%20when%20you'll%20be%20able%20to%20splash%20again (+ (Utilities|Time|GetGameTimeinSeconds) (Variables|Default|GetMinimum%20time%20between%20splashes)) _externalactor)
        (Class|PlayerPawnGenericParent|PlayerCreatesASplash _externalactor _traced_transform)
        (Class|WindManager|PlayWaterSplashParticleSystematLocation (Utilities|Casting|CastToWindManager (Fortnite|GetWindManager)) _traced_transform)))))

(event OnComponentBeginOverlap(CollisionMesh3)
  (params (OverlappedComponent :type Primitive%20Component%20Object%20Reference) (OtherActor :type Actor%20Object%20Reference) (OtherComp :type Primitive%20Component%20Object%20Reference) (OtherBodyIndex :type Integer) (bFromSweep :type Boolean) (SweepResult :type Hit%20Result%20Structure%20%28by%20ref%29))
  (bind _externalactor (Variables|Default|GetExternalActor))
  (CallFunction|HandleWaterPenetrationComponentsandArrays true OtherActor (Variables|Default|GetCollisionMesh3))
  ;[1 1 1] Early out
  (if (and (Variables|WaterSettings|GetEnable%20Player/Water%20Interactions) (Variables|Default|GetContinue))
    ;[0 0 0] Spawn A Splash Effect
    (if (and (> (Utilities|Time|GetGameTimeinSeconds) (Variables|/Game/Abilities/Player/Pawns/PlayerPawn_Generic_Parent.PlayerPawn_Generic_Parent_C|GetTime%20when%20you'll%20be%20able%20to%20splash%20again (Variables|Default|GetExternalActor))) (> (Math|Float|Absolute(Float) (.z (Transformation|GetVelocity (Variables|Default|GetExternalActor)))) 30))
      (bind (_traced_transform _velocityvector _trace_hit) (CallFunction|FindPlacementOnWaterMeshSurfaces))
      (if _trace_hit
        (Variables|/Game/Abilities/Player/Pawns/PlayerPawn_Generic_Parent.PlayerPawn_Generic_Parent_C|SetTime%20when%20you'll%20be%20able%20to%20splash%20again (+ (Utilities|Time|GetGameTimeinSeconds) (Variables|Default|GetMinimum%20time%20between%20splashes)) _externalactor)
        (Class|PlayerPawnGenericParent|PlayerCreatesASplash _externalactor _traced_transform)
        (Class|WindManager|PlayWaterSplashParticleSystematLocation (Utilities|Casting|CastToWindManager (Fortnite|GetWindManager)) _traced_transform)))))

(event OnComponentBeginOverlap(CollisionMesh4)
  (params (OverlappedComponent :type Primitive%20Component%20Object%20Reference) (OtherActor :type Actor%20Object%20Reference) (OtherComp :type Primitive%20Component%20Object%20Reference) (OtherBodyIndex :type Integer) (bFromSweep :type Boolean) (SweepResult :type Hit%20Result%20Structure%20%28by%20ref%29))
  (bind _externalactor (Variables|Default|GetExternalActor))
  (CallFunction|HandleWaterPenetrationComponentsandArrays true OtherActor (Variables|Default|GetCollisionMesh4))
  ;[1 1 1] Early out
  (if (and (Variables|WaterSettings|GetEnable%20Player/Water%20Interactions) (Variables|Default|GetContinue))
    ;[0 0 0] Spawn A Splash Effect
    (if (and (> (Utilities|Time|GetGameTimeinSeconds) (Variables|/Game/Abilities/Player/Pawns/PlayerPawn_Generic_Parent.PlayerPawn_Generic_Parent_C|GetTime%20when%20you'll%20be%20able%20to%20splash%20again (Variables|Default|GetExternalActor))) (> (Math|Float|Absolute(Float) (.z (Transformation|GetVelocity (Variables|Default|GetExternalActor)))) 30))
      (bind (_traced_transform _velocityvector _trace_hit) (CallFunction|FindPlacementOnWaterMeshSurfaces))
      (if _trace_hit
        (Variables|/Game/Abilities/Player/Pawns/PlayerPawn_Generic_Parent.PlayerPawn_Generic_Parent_C|SetTime%20when%20you'll%20be%20able%20to%20splash%20again (+ (Utilities|Time|GetGameTimeinSeconds) (Variables|Default|GetMinimum%20time%20between%20splashes)) _externalactor)
        (Class|PlayerPawnGenericParent|PlayerCreatesASplash _externalactor _traced_transform)
        (Class|WindManager|PlayWaterSplashParticleSystematLocation (Utilities|Casting|CastToWindManager (Fortnite|GetWindManager)) _traced_transform)))))

(event OnComponentEndOverlap(CollisionMesh1)
  (params (OverlappedComponent :type Primitive%20Component%20Object%20Reference) (OtherActor :type Actor%20Object%20Reference) (OtherComp :type Primitive%20Component%20Object%20Reference) (OtherBodyIndex :type Integer))
  (bind _externalactor (Variables|Default|GetExternalActor))
  (CallFunction|HandleWaterPenetrationComponentsandArrays false OtherActor (Variables|Default|GetCollisionMesh1))
  ;[1 1 1] Early out
  (if (and (Variables|WaterSettings|GetEnable%20Player/Water%20Interactions) (Variables|Default|GetContinue))
    ;[0 0 0] Spawn A Splash Effect
    (if (and (> (Utilities|Time|GetGameTimeinSeconds) (Variables|/Game/Abilities/Player/Pawns/PlayerPawn_Generic_Parent.PlayerPawn_Generic_Parent_C|GetTime%20when%20you'll%20be%20able%20to%20splash%20again (Variables|Default|GetExternalActor))) (> (Math|Float|Absolute(Float) (.z (Transformation|GetVelocity (Variables|Default|GetExternalActor)))) 30))
      (bind (_traced_transform _velocityvector _trace_hit) (CallFunction|FindPlacementOnWaterMeshSurfaces))
      (if _trace_hit
        (Variables|/Game/Abilities/Player/Pawns/PlayerPawn_Generic_Parent.PlayerPawn_Generic_Parent_C|SetTime%20when%20you'll%20be%20able%20to%20splash%20again (+ (Utilities|Time|GetGameTimeinSeconds) (Variables|Default|GetMinimum%20time%20between%20splashes)) _externalactor)
        (Class|PlayerPawnGenericParent|PlayerCreatesASplash _externalactor _traced_transform)
        (Class|WindManager|PlayWaterSplashParticleSystematLocation (Utilities|Casting|CastToWindManager (Fortnite|GetWindManager)) _traced_transform)))))

(event OnComponentEndOverlap(CollisionMesh2)
  (params (OverlappedComponent :type Primitive%20Component%20Object%20Reference) (OtherActor :type Actor%20Object%20Reference) (OtherComp :type Primitive%20Component%20Object%20Reference) (OtherBodyIndex :type Integer))
  (bind _externalactor (Variables|Default|GetExternalActor))
  (CallFunction|HandleWaterPenetrationComponentsandArrays false OtherActor (Variables|Default|GetCollisionMesh2))
  ;[1 1 1] Early out
  (if (and (Variables|WaterSettings|GetEnable%20Player/Water%20Interactions) (Variables|Default|GetContinue))
    ;[0 0 0] Spawn A Splash Effect
    (if (and (> (Utilities|Time|GetGameTimeinSeconds) (Variables|/Game/Abilities/Player/Pawns/PlayerPawn_Generic_Parent.PlayerPawn_Generic_Parent_C|GetTime%20when%20you'll%20be%20able%20to%20splash%20again (Variables|Default|GetExternalActor))) (> (Math|Float|Absolute(Float) (.z (Transformation|GetVelocity (Variables|Default|GetExternalActor)))) 30))
      (bind (_traced_transform _velocityvector _trace_hit) (CallFunction|FindPlacementOnWaterMeshSurfaces))
      (if _trace_hit
        (Variables|/Game/Abilities/Player/Pawns/PlayerPawn_Generic_Parent.PlayerPawn_Generic_Parent_C|SetTime%20when%20you'll%20be%20able%20to%20splash%20again (+ (Utilities|Time|GetGameTimeinSeconds) (Variables|Default|GetMinimum%20time%20between%20splashes)) _externalactor)
        (Class|PlayerPawnGenericParent|PlayerCreatesASplash _externalactor _traced_transform)
        (Class|WindManager|PlayWaterSplashParticleSystematLocation (Utilities|Casting|CastToWindManager (Fortnite|GetWindManager)) _traced_transform)))))

(event OnComponentEndOverlap(CollisionMesh3)
  (params (OverlappedComponent :type Primitive%20Component%20Object%20Reference) (OtherActor :type Actor%20Object%20Reference) (OtherComp :type Primitive%20Component%20Object%20Reference) (OtherBodyIndex :type Integer))
  (bind _externalactor (Variables|Default|GetExternalActor))
  (CallFunction|HandleWaterPenetrationComponentsandArrays false OtherActor (Variables|Default|GetCollisionMesh3))
  ;[1 1 1] Early out
  (if (and (Variables|WaterSettings|GetEnable%20Player/Water%20Interactions) (Variables|Default|GetContinue))
    ;[0 0 0] Spawn A Splash Effect
    (if (and (> (Utilities|Time|GetGameTimeinSeconds) (Variables|/Game/Abilities/Player/Pawns/PlayerPawn_Generic_Parent.PlayerPawn_Generic_Parent_C|GetTime%20when%20you'll%20be%20able%20to%20splash%20again (Variables|Default|GetExternalActor))) (> (Math|Float|Absolute(Float) (.z (Transformation|GetVelocity (Variables|Default|GetExternalActor)))) 30))
      (bind (_traced_transform _velocityvector _trace_hit) (CallFunction|FindPlacementOnWaterMeshSurfaces))
      (if _trace_hit
        (Variables|/Game/Abilities/Player/Pawns/PlayerPawn_Generic_Parent.PlayerPawn_Generic_Parent_C|SetTime%20when%20you'll%20be%20able%20to%20splash%20again (+ (Utilities|Time|GetGameTimeinSeconds) (Variables|Default|GetMinimum%20time%20between%20splashes)) _externalactor)
        (Class|PlayerPawnGenericParent|PlayerCreatesASplash _externalactor _traced_transform)
        (Class|WindManager|PlayWaterSplashParticleSystematLocation (Utilities|Casting|CastToWindManager (Fortnite|GetWindManager)) _traced_transform)))))

(event OnComponentEndOverlap(CollisionMesh4)
  (params (OverlappedComponent :type Primitive%20Component%20Object%20Reference) (OtherActor :type Actor%20Object%20Reference) (OtherComp :type Primitive%20Component%20Object%20Reference) (OtherBodyIndex :type Integer))
  (bind _externalactor (Variables|Default|GetExternalActor))
  (CallFunction|HandleWaterPenetrationComponentsandArrays false OtherActor (Variables|Default|GetCollisionMesh4))
  ;[1 1 1] Early out
  (if (and (Variables|WaterSettings|GetEnable%20Player/Water%20Interactions) (Variables|Default|GetContinue))
    ;[0 0 0] Spawn A Splash Effect
    (if (and (> (Utilities|Time|GetGameTimeinSeconds) (Variables|/Game/Abilities/Player/Pawns/PlayerPawn_Generic_Parent.PlayerPawn_Generic_Parent_C|GetTime%20when%20you'll%20be%20able%20to%20splash%20again (Variables|Default|GetExternalActor))) (> (Math|Float|Absolute(Float) (.z (Transformation|GetVelocity (Variables|Default|GetExternalActor)))) 30))
      (bind (_traced_transform _velocityvector _trace_hit) (CallFunction|FindPlacementOnWaterMeshSurfaces))
      (if _trace_hit
        (Variables|/Game/Abilities/Player/Pawns/PlayerPawn_Generic_Parent.PlayerPawn_Generic_Parent_C|SetTime%20when%20you'll%20be%20able%20to%20splash%20again (+ (Utilities|Time|GetGameTimeinSeconds) (Variables|Default|GetMinimum%20time%20between%20splashes)) _externalactor)
        (Class|PlayerPawnGenericParent|PlayerCreatesASplash _externalactor _traced_transform)
        (Class|WindManager|PlayWaterSplashParticleSystematLocation (Utilities|Casting|CastToWindManager (Fortnite|GetWindManager)) _traced_transform)))))

(fn Construction
  (bind _staticmesh1 (Variables|Default|GetStaticMesh1))
  (if (not (Variables|Default|GetHasExecutedConstuctionScript))
    (Variables|Default|SetHasExecutedConstuctionScript true)
    ;[1 1 1] deletes components or adds them to the struct array
    (CallFunction|SetupMeshCollisionVolumes (Variables|Default|GetCollisionMesh1))
    (CallFunction|SetupMeshCollisionVolumes (Variables|Default|GetCollisionMesh2) 2)
    (CallFunction|SetupMeshCollisionVolumes (Variables|Default|GetCollisionMesh3) 3)
    (CallFunction|SetupMeshCollisionVolumes (Variables|Default|GetCollisionMesh4) 4)
    ;[1 1 1] we trace against a single object and the trace no longer works if the model is hidden so instead we hide it with a material
    (if (not (Variables|Default|GetShowTestMesh))
      (for _index (range (Rendering|Material|GetNumMaterials _staticmesh1))
        (Rendering|Material|SetMaterial _staticmesh1 _index "/Game/Packages/Fortress_SharedMaterials/Base_Material/ShrinkModelInVertexShader.ShrinkModelInVertexShader"))))
  (return))

(fn FindPlacementonwatermeshsurfaces
  (bind _externalactor (Variables|Default|GetExternalActor))
  (bind (_hitlocation _hitnormal _bonename _outhit _returnvalue) (Collision|LineTraceComponent (select (Variables|Default|GetIs%20Deep%20Water%20Asset?) (Variables|Default|GetCollisionMesh1) (Variables|Default|GetStaticMesh1)) (+ (Transformation|GetWorldLocation (Variables|Transformation|GetRootComponent _externalactor)) "0, 0,180.000000") (+ (Transformation|GetWorldLocation (Variables|Transformation|GetRootComponent _externalactor)) "0.000000, 0,-100.000000") true false false))
  (return (Math|Transform|MakeTransform _hitlocation (Math|Rotator|MakeRotfromXZ (Math|Vector|Normalize (Math|Vector|SelectVector (Math|Vector|CrossProduct _hitnormal "1.000000, 0, 0") (Math|Vector|CrossProduct _hitnormal (Transformation|GetVelocity _externalactor)) (== (Math|Vector|VectorLength (Math|Vector|CrossProduct _hitnormal (Transformation|GetVelocity _externalactor))) 0.0))) _hitnormal)) (Transformation|GetVelocity _externalactor) _returnvalue))

(fn HandleWaterPenetrationComponentsAndArrays
  (params (In%20T%20Out%20F :type Boolean) (Touching%20Actor :type Actor%20Object%20Reference) (Touching%20Volume :type Static%20Mesh%20Component%20Object%20Reference))
  (Variables|Default|SetInternal%20Mesh Touching%20Volume)
  (Variables|Default|SetEntering%20True%20Exiting%20False In%20T%20Out%20F)
  ;[0 0 0] Find index
  (bind _asplayer_pawn_generic_parent (Utilities|Casting|CastToPlayerPawn_Generic_Parent Touching%20Actor)
    (:then
      ;[0.1351 0.4827 0.865] Track the players global number of water volumes on the character itself
      (Variables|Default|SetContinue true)
      (Variables|Default|SetExternalActor _asplayer_pawn_generic_parent)
      (Variables|/Game/Abilities/Player/Pawns/PlayerPawn_Generic_Parent.PlayerPawn_Generic_Parent_C|SetCurrentWaterMeshActor self _asplayer_pawn_generic_parent)
      (bind _output_get (Variables|/Game/Abilities/Player/Pawns/PlayerPawn_Generic_Parent.PlayerPawn_Generic_Parent_C|SetWaterCounter (+ (Variables|/Game/Abilities/Player/Pawns/PlayerPawn_Generic_Parent.PlayerPawn_Generic_Parent_C|GetWaterCounter _asplayer_pawn_generic_parent) (select (Variables|Default|GetEntering%20True%20Exiting%20False) 1 -1)) _asplayer_pawn_generic_parent))
      (bind _output_get_2 (Variables|Default|SetCurrent%20Characters%20Personal%20Water%20Counter _output_get))
      (Variables|/Game/Abilities/Player/Pawns/PlayerPawn_Generic_Parent.PlayerPawn_Generic_Parent_C|SetIsInWater (> _output_get_2 0) _asplayer_pawn_generic_parent)
      (Class|PlayerPawnGenericParent|Exited/EnteredWaterVolume _asplayer_pawn_generic_parent))
    (:CastFailed
      (Variables|Default|SetContinue false))))

(fn SetupMeshCollisionVolumes
  (params (StaticMeshComponent :type Static%20Mesh%20Component%20Object%20Reference) (Target%20Int :type Integer :default 1))
  (Utilities|IsValid StaticMeshComponent
    (:Is%20Valid
      (if (not (>= (Variables|WaterSettings|GetNumber%20of%20Collision%20Meshes) Target%20Int))
        (Components|DestroyComponent StaticMeshComponent)))
    (:Is%20Not%20Valid
      _))
  (return))

(fn ConstructionScript
  (CallFunction|Construction))
